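; Heist base manager actor that tracks guard alertness for Clyde heist music and carries base lockdown, player tag and analytics logic.
; Pawn alert counts and sets of pawns targeting the local player select a music alert level, with downgrades delayed by a buffer timer.
; Helpers stim nearby guards, reserve AI density, toggle global lockdown material visuals and send end-game analytics.
; Lockdown request, trigger and overlap handlers are empty.
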(event EventBeginPlay
  (|CallParentFunction))

(event Custom|OnRequestLockdownBase
  (params (Payload :type Struct%20Heist%20Bases%20Lockdown%20Initiated%20Payload%20Structure)))

(event Custom|AddBasePawn
  ; Initial setup of pawn for tracking Alerted Level
  (Utilities|Map|Add (Variables|ClydeBaseMusic|GetPawnAlertedLevels) "Unaware" (+ (Utilities|Map|Find (Variables|ClydeBaseMusic|GetPawnAlertedLevels) "Unaware") 1)))

(event Custom|EndLockdownByExpiration
  (Variables|Lockdown|SetbInLockdown false)
  (Variables|Lockdown|SetbLockdownExpired true))

(event Custom|RequestClydeAlertLevelChange
  (params (Requested%20Alert%20Level :type Enum_ClydeAlertLevels%20Enum :default "NewEnumerator0"))
  ; If request is for a higher level of alertness change music accrodingly, if it is for a lower level of alertness request change with buffer in case it is a quick toggle back to the higher level
  (if (>= (Math|Conversions|ToInteger(Byte) Requested%20Alert%20Level) (Math|Conversions|ToInteger(Byte) (Variables|ClydeBaseMusic|GetBaseClydeAlertLevel)))
    (Utilities|Time|ClearandInvalidateTimerbyHandle (Variables|ClydeBaseMusic|GetRequestedClydeLowerAlertLevelTimer))
    (if (Utilities|Enum|NotEqual(Enum) Requested%20Alert%20Level (Variables|ClydeBaseMusic|GetBaseClydeAlertLevel))
      (bind _output_get (Variables|ClydeBaseMusic|SetBaseClydeAlertLevel Requested%20Alert%20Level))
      (Development|PrintString (Utilities|String|Append "Set Base Clyde Alert Level: " (Utilities|String|EnumtoString _output_get)) true true "(R=0.000000,G=0.660000,B=1.000000,A=1.000000)" 2e+01))
    (else
      (Variables|ClydeBaseMusic|SetRequestedBaseClydeAlertLevel Requested%20Alert%20Level)
      (if (not (Utilities|Time|IsValidTimerHandle (Variables|ClydeBaseMusic|GetRequestedClydeLowerAlertLevelTimer)))
        (bind _returnvalue (Utilities|Time|SetTimerbyEvent (EventDispatchers|CreateEvent :!event Custom|SetRequestedClydeAlertLevel) (Variables|ClydeBaseMusic|GetClydeAlertLevelChangeBufferTime)))
        (Variables|ClydeBaseMusic|SetRequestedClydeLowerAlertLevelTimer _returnvalue)))))

(event Custom|UpdatePawnAlertLevel
  (params (OldAlertLevel :type EAlertLevel%20Enum :default "Unaware") (NewAlertLevel :type EAlertLevel%20Enum :default "Unaware") (TrackingLocalPlayer :type Boolean) (TrackingPawn :type BP%20Player%20Pawn%20Non%20Participant%20Object%20Reference))
  (bind (_value _returnvalue) (Utilities|Map|Find (Variables|ClydeBaseMusic|GetPawnAlertedLevels) OldAlertLevel))
  (bind _pawnalertedlevels (Variables|ClydeBaseMusic|GetPawnAlertedLevels))
  (bind (_value_2 _returnvalue_2) (Utilities|Map|Find _pawnalertedlevels NewAlertLevel))
  (bind _pawnalertedlevels_2 (Variables|ClydeBaseMusic|GetPawnAlertedLevels))
  (CallFunction|AdjustLocalPlayerTracking TrackingLocalPlayer TrackingPawn NewAlertLevel)
  ; Add henchmen to alert state tracking
  (Utilities|FlowControl|Sequence
    (:then_0
      (if _returnvalue_2
        (Utilities|Map|Add _pawnalertedlevels NewAlertLevel (+ _value_2 1))
        (else
          _))
      (if _returnvalue
        (Utilities|Map|Add _pawnalertedlevels OldAlertLevel (- _value 1))))
    (:then_1
      ; Check if a change was made that could require a heist music change
      (bind _keys (Utilities|Map|Keys _pawnalertedlevels_2))
      (bind _array_element (Utilities|Array|ForEachLoopwithBreak _keys)
        (:LoopBody
          (if (.ReturnValue (Utilities|Map|Find _pawnalertedlevels_2 _array_element))
            (if (== (Utilities|Map|Find _pawnalertedlevels_2 _array_element) 0)
              (if (Utilities|Enum|Equal(Enum) OldAlertLevel _array_element)
                (CallFunction|CheckBaseAlertLevel)))
            (if (== (Utilities|Map|Find _pawnalertedlevels_2 _array_element) 1)
              (if (Utilities|Enum|Equal(Enum) _array_element NewAlertLevel)
                _))))))))

(event Custom|AdjustLocalPlayerTracking
  (params (TrackingLocalPlayer :type Boolean) (TrackingPawn :type BP%20Player%20Pawn%20Non%20Participant%20Object%20Reference) (AlertLevel :type EAlertLevel%20Enum :default "Unaware"))
  (bind _suspiciouspawns (Variables|ClydeBaseMusic|GetSuspiciousPawns))
  (bind _alertedpawns (Variables|ClydeBaseMusic|GetAlertedPawns))
  ; Track pawns that are currently targetting the local player
  (Utilities|Set|Remove _suspiciouspawns TrackingPawn)
  (Utilities|Set|Remove _alertedpawns TrackingPawn)
  (if TrackingLocalPlayer
    (switch Utilities|FlowControl|Switch|SwitchonEAlertLevel AlertLevel
      (:Suspicious
        (Utilities|Set|Add _suspiciouspawns TrackingPawn))
      (:Threatened
        (Utilities|Set|Add _alertedpawns TrackingPawn)))))

(event Custom|SetRequestedClydeAlertLevel
  (bind _requestedbaseclydealertlevel (Variables|ClydeBaseMusic|GetRequestedBaseClydeAlertLevel))
  ; Complete request for lower level of alertness heist music
  (Utilities|Time|ClearandInvalidateTimerbyHandle (Variables|ClydeBaseMusic|GetRequestedClydeLowerAlertLevelTimer))
  (if (Utilities|Enum|NotEqual(Enum) _requestedbaseclydealertlevel (Variables|ClydeBaseMusic|GetBaseClydeAlertLevel))
    (bind _output_get (Variables|ClydeBaseMusic|SetBaseClydeAlertLevel _requestedbaseclydealertlevel))
    (Development|PrintString (Utilities|String|Append "Set Requested Base Clyde Alert Level: " (Utilities|String|EnumtoString _output_get)) true true "(R=0.000000,G=0.660000,B=1.000000,A=1.000000)" 2e+01)))

(event Custom|CheckBaseAlertLevel
  (bind _pawnalertedlevels (Variables|ClydeBaseMusic|GetPawnAlertedLevels))
  ; See that change represents a valid request for a heist music change
  (bind _keys (Utilities|Map|Keys _pawnalertedlevels))
  (bind _array_element (Utilities|Array|ForEachLoopwithBreak _keys)
    (:LoopBody
      (if (.ReturnValue (Utilities|Map|Find _pawnalertedlevels _array_element))
        (if (> (Utilities|Map|Find _pawnalertedlevels _array_element) 0)
          (Variables|ClydeBaseMusic|SetBaseAlertLevel _array_element)
          (CallFunction|CheckBaseAlertLevelChange))))))

(event Custom|CheckBaseAlertLevelChange
  ; Determine music level to request from music manager
  (switch Utilities|FlowControl|Switch|SwitchonEAlertLevel (Variables|ClydeBaseMusic|GetBaseAlertLevel)
    (:Unaware
      (CallFunction|RequestClydeAlertLevelChange "NewEnumerator0"))
    (:Suspicious
      (if (> (Utilities|Set|Length (Variables|ClydeBaseMusic|GetSuspiciousPawns)) 0)
        (CallFunction|RequestClydeAlertLevelChange "NewEnumerator3")
        (else
          (CallFunction|RequestClydeAlertLevelChange "NewEnumerator1"))))
    (:Threatened
      (if (> (Utilities|Set|Length (Variables|ClydeBaseMusic|GetAlertedPawns)) 0)
        (CallFunction|RequestClydeAlertLevelChange "NewEnumerator4")
        (else
          (CallFunction|RequestClydeAlertLevelChange "NewEnumerator2"))))))

(event OnComponentBeginOverlap(HeistBaseAreaCollision)
  (params (OverlappedComponent :type Primitive%20Component%20Object%20Reference) (OtherActor :type Actor%20Object%20Reference) (OtherComp :type Primitive%20Component%20Object%20Reference) (OtherBodyIndex :type Integer) (bFromSweep :type Boolean) (SweepResult :type Hit%20Result%20Structure%20%28by%20ref%29)))

(event Custom|CallMusicManagerThemeChange)

(event OnComponentEndOverlap(HeistBaseAreaCollision)
  (params (OverlappedComponent :type Primitive%20Component%20Object%20Reference) (OtherActor :type Actor%20Object%20Reference) (OtherComp :type Primitive%20Component%20Object%20Reference) (OtherBodyIndex :type Integer)))

(event EventEndPlay
  (params (EndPlayReason :type EEndPlayReason%20Enum :default "Destroyed"))
  (bind _lockdown_request_gameplaymessagetask (Variables|Default|GetLockdown%20Request%20GameplayMessageTask))
  ;[1 0.25 0] END PLAY
  ; Cleanup on endplay: timers, bind events, etc.
  (Utilities|FlowControl|Sequence
    (:then_0
      ;;[1 0.25 0] Sequence[0] General Cleanup
      (Protected|Utility|EndPlayCleanup)
      (Networking|SwitchHasAuthority
        (:Authority
          ;;[1 0 0.8325] LOG: Log the end event - debugging why Lockdown isn't consistently triggered
          (Fortnite|FortShippingLog self (Utilities|String|Append "HEIST BASE: Manager End Play Called. Reason = " (Utilities|String|EnumtoString EndPlayReason))))))
    (:then_1
      ;;[1 0.25 0] Sequence[1] Clear and Invalidate Timers
      (Utilities|Time|ClearandInvalidateTimerbyHandle (Variables|Lockdown|GetTH_LockdownExpirationTimer)))
    (:then_2
      ;;[1 0.25 0] Sequence[2] - Unbind Events
      (Utilities|IsValid _lockdown_request_gameplaymessagetask
        (:Is%20Valid
          (EventDispatchers|UnbindallEventsfromOnMessageReceived _lockdown_request_gameplaymessagetask))
        (:Is%20Not%20Valid
          _))
      (Variables|Default|GetCachedActorRegistryDataObject
        (:then)
        (:else
          (bind _gamephaseupdatedasynctask (Variables|Default|GetGamePhaseUpdatedAsyncTask)
            (:then
              (GameplayEventRouter|StopListeningtoStatefulEvent _gamephaseupdatedasynctask))
            (:else)))))
    (:then_3
      ;;[1 0.25 0] Sequence[3] - Call parent EndPlay last so inherited cleanup isn't run before other cleanup
      (|CallParentFunction EndPlayReason))))

(fn AddorRemovePlayer
  (params (PlayerActor :type Actor%20Object%20Reference) (AddPlayer :type Boolean))
  (bind _array (Utilities|Array|MakeArray))
  (bind _returnvalue (GameplayTags|MakeGameplayTagContainerfromArray _array))
  (bind _playerpawnsinheistbase (Variables|Default|GetPlayerPawnsInHeistBase))
  (bind _replicatepawntags_const (Variables|Default|GetReplicatePawnTags%20Const))
  (bind _asfort_player_pawn_athena (Utilities|Casting|CastToFortPlayerPawnAthena PlayerActor))
  (if AddPlayer
    (bind _returnvalue_2 (Utilities|Array|AddUnique _playerpawnsinheistbase _asfort_player_pawn_athena))
    (if (!= _returnvalue_2 -1)
      (Ability|GameplayEffect|AddLooseGameplayTags _asfort_player_pawn_athena _returnvalue _replicatepawntags_const) ; We successfully added this player to our list, give the player tags for this heist base and its current status.
      ;[0.2435 0.1798 0.4062] Analytics: Store players and the time they enter in order to get a total count and completion times for the base
      (if (not (Pawn|IsBotControlled _asfort_player_pawn_athena))
        (Utilities|Map|Add (Variables|Analytics|GetAnalyticsMapPlayerToEntryTime) _asfort_player_pawn_athena (Utilities|Time|GetGameTimeinSeconds))))
    (else
      (bind _returnvalue_3 (Utilities|Array|RemoveItem _playerpawnsinheistbase _asfort_player_pawn_athena))
      (if _returnvalue_3
        (Ability|GameplayEffect|RemoveLooseGameplayTags _asfort_player_pawn_athena _returnvalue _replicatepawntags_const) ; We successfully removed this player to our list, remove from this player the tags for this heist base and its current status.
        ;[0.2435 0.1798 0.4062] Analytics: Store when the lockdown-initiating player escapes
        (bind _lockdowntriggeringpawnathena (Variables|Analytics|GetLockdownTriggeringPawnAthena)
          (:then
            (if (== _asfort_player_pawn_athena _lockdowntriggeringpawnathena)
              (Variables|Analytics|SetLockdownPlayerEscapeTime (- (Utilities|Time|GetGameTimeinSeconds) (Utilities|Map|Find (Variables|Analytics|GetAnalyticsMapPlayerToEntryTime) _lockdowntriggeringpawnathena)))
              ;[0.2435 0.1798 0.4062] Analytics: Is base within storm when the lockdown-initiating player escapes the base?
              (bind _returnvalue_4 (SafeZone|IsLocationinSafeZone (Transformation|GetActorLocation)))
              (Variables|Analytics|SetAnalytic_BaseInStormOnEscape (not _returnvalue_4))))
          (:else))))))

(fn CacheRegisteredBaseDefense
  (params (RegisteredActor :type Actor%20Object%20Reference))
  ;[0.4973 0.715 1] CacheRegisteredBaseDefense
  ; Add the registered defense actor to this heist base - tag was already checked by the registry.
  (Utilities|Array|AddUnique (Variables|BaseDefense|GetCachedBaseDefenses) RegisteredActor))

(fn CanBeSavedToCreativeVolume
  (return false))

(fn EditorRegisterBaseDefenses
  (bind _self self)
  (bind _returnvalue (Transformation|GetActorLocation _self))
  (bind _editorbasedefensedetectionradius (Variables|LevelDesign|GetEditorBaseDefenseDetectionRadius))
  ;[0 0 0] EditorRegisterBaseDefenses
  ; Get all objects within the specified radius, and register them with the current base if they are valid defenses (has the lockdown interface and registration component)
  (bind _outhits (Collision|MultiSphereTraceForObjects (+ _returnvalue -1.0) _returnvalue _editorbasedefensedetectionradius (Utilities|Array|MakeArray "ObjectTypeQuery1" "ObjectTypeQuery2" "ObjectTypeQuery6") true)) ; get all the objects in the base radius\nrequires complex trace for objects that don't have \nsimple collision (ex. Lockdown_Spot_C)
  (for _array_element _outhits
    (if (Utilities|DoesObjectImplementInterface (.HitActor (Collision|BreakHitResult _array_element)) "/Clyde/Audio/Items/BPI_HeistBases_Lockdown.BPI_HeistBases_Lockdown_C")
      ;;[0.8203 0 1] Track count of base defenses found for logging
      (Math|Integer|IncrementInt (LocalVariables|Geth_CountBaseDefensesFound))))
  ;;[0.8886 0 1] Debug logging and visuals to confirm the registration has been successful.
  (Rendering|Debug|DrawDebugSphere _returnvalue _editorbasedefensedetectionradius 12 "(R=1.000000,G=0.000000,B=0.928250,A=1.000000)" 12.0 2e+01)
  (Fortnite|FortShippingLog self (Utilities|String|Append "HEIST BASE: Found and registered  " (Utilities|String|ToString(Integer) (LocalVariables|Geth_CountBaseDefensesFound)) " defenses and props for lockdown " (GameplayTags|GetDebugStringfromGameplayTag))))

(fn EndPlayCleanup
  (Networking|SwitchHasAuthority
    (:Authority
      (Protected|AI|ReserveAIDensityReservoir true))
    (:Remote
      (Rendering|Material|SetScalarParameterValue "/Game/Packages/Fortress_SharedMaterials/GlobalMaterialParameters/FortniteMaterialParameters.FortniteMaterialParameters" "LockdownActive" :!class /Script/Engine.KismetMaterialLibrary) ; reset lockdown visuals for all emissive materials that get Lockdown effects\nif this is not reset, other bases will also appear to be in Lockdown
    )))

(fn GenerateAnalyticsLogMessage
  (params (AnalyticAttributeName :type String) (AnalyticsValue :type String))
  ;[0.2435 0.1798 0.4062] GenerateAnalyticsLogMessage
  ; Format the Log with a standard string for analytics
  (Fortnite|FortShippingLog self (Utilities|String|Append "HEIST BASE ANALYTICS: (" "" ") will send" AnalyticAttributeName " = " AnalyticsValue)))

(fn GetAnalyticsAttributes
  (bind _analytic_baseinstormonescape (Variables|Analytics|GetAnalytic_BaseInStormOnEscape))
  (bind _lockdownplayerescapetime (Variables|Analytics|GetLockdownPlayerEscapeTime))
  (bind _analytic_waslockeddown (Variables|Analytics|GetAnalytic_WasLockedDown))
  (bind _analytic_lockdowntime (Variables|Analytics|GetAnalytic_LockdownTime))
  (bind _analytic_baseinstormonlockdown (Variables|Analytics|GetAnalytic_BaseInStormOnLockdown))
  (bind _analytic_playerlockdowntime (Variables|Analytics|GetAnalytic_PlayerLockdownTime))
  ;[0.2435 0.1798 0.4062] GetAnalyticsAttributes
  ; Get and log all tracked attributes for the Heist Base
  (Utilities|Array|Clear (LocalVariables|GetReturn_LocalAnalytics))
  (Utilities|FlowControl|Sequence
    (:then_0
      ;;[0.2435 0.1798 0.4062] Heist Base Analytics: attributes related to the base iteself and its status
      (bind _returnvalue (Analytics|ConstructAnalyticsAttribute :AttributeName "LockedDownDuringMatch" :IntValue -1 :floatValue -1.0 :BoolValue _analytic_waslockeddown)) ; LockedDownDuringMatch\nwhether the base was locked down or not during the match
      (Utilities|Array|Add (LocalVariables|GetReturn_LocalAnalytics) _returnvalue)
      (Analytics|GenerateAnalyticsLogMessage "LockedDownDuringMatch" (Utilities|String|ToString(Boolean) _analytic_waslockeddown))
      (bind _returnvalue_2 (Analytics|ConstructAnalyticsAttribute :AttributeName "LockdownInitiatedMatchTime" :IntValue -1 :floatValue _analytic_lockdowntime :BoolValue false)) ; LockdownInitiatedMatchTime\ntime when the base enters lockdown
      (Utilities|Array|Add (LocalVariables|GetReturn_LocalAnalytics) _returnvalue_2)
      (Analytics|GenerateAnalyticsLogMessage "LockdownInitiatedMatchTime" (Utilities|String|ToString(Float) _analytic_lockdowntime))
      (bind _returnvalue_3 (Analytics|ConstructAnalyticsAttribute :AttributeName "StormPresentOnLockdown" :IntValue -1 :floatValue -1.0 :BoolValue _analytic_baseinstormonlockdown)) ; StormPresentOnLockdown\nif the base manager was within the storm when lockdown was initiated
      (Utilities|Array|Add (LocalVariables|GetReturn_LocalAnalytics) _returnvalue_3)
      (Analytics|GenerateAnalyticsLogMessage "StormPresentOnLockdown" (Utilities|String|ToString(Boolean) _analytic_baseinstormonlockdown))
      (bind _returnvalue_4 (Analytics|ConstructAnalyticsAttribute :AttributeName "StormPresentOnEscape" :IntValue -1 :floatValue -1.0 :BoolValue _analytic_baseinstormonescape)) ; StormPresentOnEscape\nIf the base manager was within the storm when the lockdown initiating player escapes the base
      (Utilities|Array|Add (LocalVariables|GetReturn_LocalAnalytics) _returnvalue_4)
      (Analytics|GenerateAnalyticsLogMessage "StormPresentOnEscape" (Utilities|String|ToString(Boolean) _analytic_baseinstormonescape)))
    (:then_1
      ;;[0.2435 0.1798 0.4062] Heist Player Analytics: attributes related to the players in the base
      (bind _returnvalue_5 (Analytics|ConstructAnalyticsAttribute :AttributeName "LockdownInitiatedPlayerTime" :IntValue -1 :floatValue _analytic_playerlockdowntime :BoolValue false)) ; LockdownInitiatedPlayerTime\nTime from the player who initiates lockdown from when they\n enter the base to starting lockdown (claiming item)
      (Utilities|Array|Add (LocalVariables|GetReturn_LocalAnalytics) _returnvalue_5)
      (Analytics|GenerateAnalyticsLogMessage "LockdownInitiatedPlayerTime" (Utilities|String|ToString(Float) _analytic_playerlockdowntime))
      (bind _returnvalue_6 (Analytics|ConstructAnalyticsAttribute :AttributeName "LockdownEscapePlayerTime" :IntValue -1 :floatValue _lockdownplayerescapetime :BoolValue false)) ; LockdownEscapePlayerTime\nTime for the player who initiates lockdown from when they\n enter the base to when they escape after claiming the reward
      (Utilities|Array|Add (LocalVariables|GetReturn_LocalAnalytics) _returnvalue_6)
      (Analytics|GenerateAnalyticsLogMessage "LockdownEscapePlayerTime" (Utilities|String|ToString(Float) _lockdownplayerescapetime)))
    (:then_2
      ;;[0.9707 0.4876 1] RETURN: Send all the tracked analytic events
      (return (LocalVariables|GetReturn_LocalAnalytics)))))

(fn GetAnalyticsEventNameForBase)

(fn GetAndRegisterDefenses
  ;[0.4973 0.715 1] GetAndRegisterDefenses
  ; Find all the base defesnes for this managed base and store them for future use
  (Networking|SwitchHasAuthority))

(fn GetMaxAIToAlert
  (bind _hotfix_lockdownaialertmaxtoalert (Variables|Hotfix|GetHotfix_LockdownAIAlertMaxToAlert))
  (bind _currentplaylistmaxteamsize (Variables|AI|GetCurrentPlaylistMaxTeamSize))
  (if (ScalableFloat|GetValueasBool (Variables|Hotfix|GetHotfix_AIScalabilityTuningEnabled))
    (if (Utilities|Array|IsValidIndex _hotfix_lockdownaialertmaxtoalert _currentplaylistmaxteamsize)
      (return (ScalableFloat|GetValueasInteger (Utilities|Array|Get(acopy) _hotfix_lockdownaialertmaxtoalert _currentplaylistmaxteamsize)))
      (else
        _))
    (else
      _))
  (return (ScalableFloat|GetValueasInteger (Variables|Hotfix|GetHotfix_DefaultMaxAIToAlert))))

(fn GetTargetTagsForQuestEvent
  (return 0))

(fn HandleGamePhaseChanged
  (params (GamePhase :type EAthenaGamePhase%20Enum))
  ;[0.03332 0.3212 0.2859] HandleGamePhaseChanged
  ; When the game is ended, send the assigned objective analytics event for the match
  (if (Utilities|Enum|Equal(Enum) GamePhase "EndGame")
    ;;[0.2435 0.1798 0.4062] Get and Fire all Analytic Attributes for the match (each objective will do this)
    (bind _analyticsattributes (Analytics|GetAnalyticsAttributes))
    (Analytics|FireAnalyticsEventBlueprintArray "" _analyticsattributes)
    (bind _gamephaseupdatedasynctask (Variables|Default|GetGamePhaseUpdatedAsyncTask)
      (:then
        (GameplayEventRouter|StopListeningtoStatefulEvent _gamephaseupdatedasynctask))
      (:else))))

(fn HeistBaseEndLockdown
  (params (Location%20Id%20Substring :type String)))

(fn HeistBaseTriggerLockdown
  (params (Location%20Id%20Substring :type String)))

(fn OnRep_bInLockdown
  (bind _binlockdown (Variables|Lockdown|GetbInLockdown))
  (Fortnite|FortShippingLog self (Utilities|String|Append "OnRepBInLockDown() [" (Utilities|GetDisplayName self) "] bInLockdown: [" (Utilities|String|ToString(Boolean) _binlockdown) "]."))
  (Networking|CanExecuteCosmeticEvents
    (:True
      (if _binlockdown
        (Rendering|Material|SetScalarParameterValue "/Game/Packages/Fortress_SharedMaterials/GlobalMaterialParameters/FortniteMaterialParameters.FortniteMaterialParameters" "LockdownActive" (Math|Conversions|ToFloat(Boolean) _binlockdown) :!class /Script/Engine.KismetMaterialLibrary) ; turn on lockdown visuals when locked down
      ))
    (:False
      (if (Variables|Lockdown|GetbInLockdown)
        (Variables|Analytics|SetAnalytic_WasLockedDown true)))))

(fn OnRep_bLockdownExpired
  (Networking|CanExecuteCosmeticEvents
    (:True
      (if (Variables|Lockdown|GetbLockdownExpired)
        (Rendering|Material|SetScalarParameterValue "/Game/Packages/Fortress_SharedMaterials/GlobalMaterialParameters/FortniteMaterialParameters.FortniteMaterialParameters" "LockdownActive" (Math|Conversions|ToFloat(Boolean) (not (ScalableFloat|GetValueasBool (Variables|Hotfix|GetHF_LockdownExpirationDisablesVisuals)))) :!class /Script/Engine.KismetMaterialLibrary) ; set lockdown emissive visuals when expired
      ))))

(fn ReserveAIDensityReservoir
  (params (bUnreserve :type Boolean))
  (bind _self self)
  (Networking|SwitchHasAuthority
    (:Authority
      (bind _livingworldmanager (Utility|GetValidLivingWorldManager)
        (:Is%20Valid
          (if (not bUnreserve)
            (Class|FortAthenaLivingWorldManager|RegisterDensityReservoir _livingworldmanager _self (Variables|AI|Budget|GetReservedReservoirDensity) (Variables|AI|Budget|GetReservedReservoirRange))
            (Fortnite|FortShippingLog _self "Registered density reservoir")
            (else
              (Class|FortAthenaLivingWorldManager|UnregisterDensityReservoir _livingworldmanager _self)
              (Fortnite|FortShippingLog _self "Unregistered self from density reservoir"))))))))

(fn SendLockdownStimToGuards
  (Networking|SwitchHasAuthority
    (:Authority
      (LocalVariables|SetLMaxAIToAlert (CallFunction|GetMaxAIToAlert))
      ;[0.1708 1 0.2319] Set Alert location to the world location of the base manager and the offset
      (bind _output_get (LocalVariables|SetLLockdownAlertLocation (+ (Variables|Lockdown|GetLockdownAlertLocationOffset) (Transformation|GetActorLocation))))
      ;[0.1407 0.1494 1] Trace for pawn capsules in an area around the lockdown locaiton
      (bind _outhits (Collision|MultiSphereTraceByChannel :Start _output_get :End (+ _output_get "0.000000,0.000000,1.000000") :Radius (ScalableFloat|GetValueatLevel (Variables|Hotfix|GetHotfix_LockdownAIAlertRange)) :TraceChannel "TraceTypeQuery9" :bTraceComplex false :DrawDebugType "ForDuration" :bIgnoreSelf true :TraceColor "(R=1.000000,G=0.000000,B=0.000000,A=1.000000)" :TraceHitColor "(R=0.000000,G=1.000000,B=0.000000,A=1.000000)" :DrawTime 5.0))
      ;[0.1407 0.1494 1] Send alert stimulus to base ai, limiting by count and the closest ai (hitresult actors are sorted by distance along the trace. I tested and verified this)
      (bind _array_element (Utilities|Array|ForEachLoopwithBreak _outhits)
        (:LoopBody
          (if (< (LocalVariables|GetLAlertedGuardsCount) (LocalVariables|GetLMaxAIToAlert))
            _
            (else
              _))
          (bind _asfort_player_pawn (Utilities|Casting|CastToFortPlayerPawn (.HitActor (Collision|BreakHitResult _array_element))))
          (if (GameplayTags|HasTag :TagContainer (Fortnite|Factions|GetFactionTagContainerforActor _asfort_player_pawn) :bExactMatch true)
            (bind _asfort_athena_aibot_controller (Utilities|Casting|CastToFortAthenaAIBotController (Pawn|GetController _asfort_player_pawn)))
            (Class|FortAthenaAIBotController|QueueStim _asfort_athena_aibot_controller self (LocalVariables|GetLLockdownAlertLocation) (Math|Vector|GetUnitDirection(Vector) (Transformation|GetActorLocation _asfort_player_pawn) (LocalVariables|GetLLockdownAlertLocation)) "Enemy" 1.0 "None")
            (Math|Integer|IncrementInt (LocalVariables|GetLAlertedGuardsCount))))
        (:Completed
          _))
      ;[1 0.3042 0.1474] Log alert results
      (Fortnite|FortShippingLog self (Utilities|String|Append (Utilities|String|ToString(Name) "None") "Lockdown alerted guards count = " (Utilities|String|ToString(Integer) (LocalVariables|GetLAlertedGuardsCount)))))))

(fn SetStatus
  (bind _replicatepawntags_const (Variables|Default|GetReplicatePawnTags%20Const))
  (if (!= 0 0)
    (Utilities|FlowControl|Sequence
      (:then_0
        ;[0.1874 0.9901 1] Update the status tags of all players within this heist base.
        (for _array_element (Variables|Default|GetPlayerPawnsInHeistBase)
          (Utilities|IsValid _array_element
            (:Is%20Valid
              (Ability|GameplayEffect|RemoveLooseGameplayTags _array_element (GameplayTags|MakeGameplayTagContainerfromTag) _replicatepawntags_const)
              (Ability|GameplayEffect|AddLooseGameplayTags _array_element (GameplayTags|MakeGameplayTagContainerfromTag) _replicatepawntags_const)))))
      (:then_1))))

(fn UnRegisterDefenses
  (Utilities|Array|Clear (Variables|BaseDefense|GetCachedBaseDefenses)))

(fn ConstructionScript)
